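# Material function for fluffy fabric character materials that recolours base colour with primary and secondary tints, behind the Use Rich Color static switch.
# The primary tint multiplies contrast-adjusted, influence-scaled base colour, with boost, hue shift and desaturation toggled by On/Off lerp parameters.
# The secondary tint supports only influence and desaturation.
# Red then green mask channels layer the tints, using either a customization texture or the Tech Art Mask input.
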MaterialFunction /Game/Global/Materials/Character_Parent_2023/FluffyFabric_Parent_2023/MF_FluffyFabric_RichColor
graph {nodes [
  n1 = FunctionInput() {inputName=Result, inputType=FunctionInput_MaterialAttributes}
  n2 = GetMaterialAttributes(None=@n1) {attributeGetTypes [#1, #2]}
  n3 = ScalarParameter() {defaultValue=1, parameterName=Primary Color Influence}
  n4 = ScalarParameter() {defaultValue=1, parameterName=Color Boost}
  n5 = Multiply(
   A=Power(Base=@n2.BaseColor, Exp=ScalarParameter() {defaultValue=1, parameterName=Primary Color - Contrast})
   B=LinearInterpolate(
    A=@n3
    B=Multiply(A=@n3, B=@n4)
    Alpha=ScalarParameter() {controlType=Enumeration, enumeration=/Game/Global/Materials/Enums/E_Mat_OnOff, parameterName=Primary Color - Use Color Boost}
   )
  )
  n6 = VectorParameter() {parameterName=Primary Color, defaultValue=<rgba 1 1 1 1>}
  n7 = LinearInterpolate(
   A=@n6.RGB
   B=MaterialFunctionCall(Hue Shift Percentage=ScalarParameter() {parameterName=Primary Color - Hue Shift}, Texture=@n6.RGB) {materialFunction=&HueShift}.Result
   Alpha=ScalarParameter() {controlType=Enumeration, enumeration=&E_Mat_OnOff, parameterName=Primary Color - Use Hue Shift}
  )
  n8 = LinearInterpolate(
   A=@n7
   B=Desaturation(None=@n7, Fraction=ScalarParameter() {parameterName=Primary Color - Desaturation})
   Alpha=ScalarParameter() {controlType=Enumeration, enumeration=&E_Mat_OnOff, parameterName=Primary Color - Use Desaturation}
  )
  n9 = TextureSampleParameter2D() {parameterName=Color Customization Mask, samplerType=SAMPLERTYPE_Masks, texture=&T_BlackMask}
  n10 = StaticSwitchParameter(True=@n9.RGB, False=FunctionInput() {inputName=Tech Art Mask, sortPriority=10}) {parameterName=Use Color Customization Mask}
  n11 = ScalarParameter() {defaultValue=1, parameterName=Secondary Color Influence}
  n12 = VectorParameter() {parameterName=Secondary Color, defaultValue=<rgba 1 1 1 1>}
  n13 = ScalarParameter() {parameterName=Secondary Color - Desaturation}
  n14 = LinearInterpolate(
   A=LinearInterpolate(A=@n2.BaseColor, B=Multiply(A=@n5, B=@n8), Alpha=ComponentMask(None=@n10) {r=true})
   B=Multiply(A=Multiply(A=@n2.BaseColor, B=@n11), B=Desaturation(None=@n12.RGB, Fraction=@n13))
   Alpha=ComponentMask(None=@n10) {g=true}
  )
  FunctionOutput(None=StaticSwitchParameter(
    True=SetMaterialAttributes!(MaterialAttributes=@n2.MaterialAttributes, BaseColor=@n14) {attributeSetTypes [#1]}
    False=@n1
   ) {parameterName=Use Rich Color}) {outputName=SetMaterialAttributes}
 ]}
